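# Control Rig for the LiftingRaysHat female medium character's face accessory that corrects hair and hat bones as the body poses.
# A branch on wearing back bling and cape selects Constrains Backpack or Setup Hair.
# The backpack path parent-constrains back hair to spine, neck and shoulder bones; both paths apply corrective transform offsets.
# Spherical pose readers, remapped bone rotations and movement, weapon and emote variables weight these offsets, often smoothed by interpolation.
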ControlRigBlueprint /BRCosmetics/Characters/Player/Female/Medium/Bodies/F_MED_LiftingRaysHat/Meshes/Parts/F_MED_LiftingRaysHat_FaceAcc_CtrlRig
graphs [
 Multiply_6_ContainedGraph {nodes [
   n1 = RigVMFunctionEntryNode()
   RigVMFunctionReturnNode(Result=&RigVMFunction_MathDoubleMul(B=@n1.C, A=&RigVMFunction_MathDoubleMul(A=@n1.A, B=@n1.B).Result).Result)
  ]}
 Multiply_7_ContainedGraph {nodes [
   n1 = RigVMFunctionEntryNode()
   RigVMFunctionReturnNode(Result=&RigVMFunction_MathDoubleMul(B=@n1.C, A=&RigVMFunction_MathDoubleMul(A=@n1.A, B=@n1.B).Result).Result)
  ]}
 Multiply_7_ContainedGraph {nodes [
   n1 = RigVMFunctionEntryNode()
   n2 = &RigVMFunction_MathDoubleMul(A=@n1.A, B=@n1.B)
   RigVMFunctionReturnNode(Result=&RigVMFunction_MathDoubleMul(B=@n1.D, A=&RigVMFunction_MathDoubleMul(B=@n1.C, A=@n2.Result).Result).Result)
  ]}
 Multiply_9_ContainedGraph {nodes [
   n1 = RigVMFunctionEntryNode()
   n2 = &RigVMFunction_MathDoubleMul(A=@n1.A, B=@n1.B)
   RigVMFunctionReturnNode(Result=&RigVMFunction_MathDoubleMul(B=@n1.D, A=&RigVMFunction_MathDoubleMul(B=@n1.C, A=@n2.Result).Result).Result)
  ]}
 New Function_ContainedGraph {nodes [
   n1 = RigVMFunctionEntryNode()
   RigVMFunction_MathQuaternionToEuler = &RigVMFunction_MathQuaternionToEuler(
    RotationOrder=ZYX
    Value=&RigUnit_GetRelativeTransformForItem(
     Child=(Type=Bone,Name="None")
     Parent=(Type=Bone,Name="None")
     bChildInitial=False
     bParentInitial=False
     Child.Name=@n1.Child
     Parent.Name=@n1.Parent
    ).RelativeTransform.Rotation
   )
   Remap = &RigVMFunction_MathDoubleRemap(
    TargetMaximum=1
    TargetMinimum=0
    Value=@RigVMFunction_MathQuaternionToEuler.Result.Y
    SourceMinimum=@n1.SourceMin
    SourceMaximum=@n1.SourceMax
   )
   > RigVMFunctionReturnNode(Delta Y=@Remap.Result)
  ]}
 New Function_ContainedGraph {nodes [
   RigVMFunctionEntryNode()
   > RigVMFunction_Sequence = RigVMAggregateNode()
   &RigUnit_ParentConstraint(
    AdvancedSettings=(RotationOrderForFilter=XZY)
    Child=(Type=Bone,Name="dyn_hairB_C_02")
    Parents=((Item=(Type=Bone,Name="spine_03"),Weight=0.0111),(Item=(Type=Bone,Name="neck_01"),Weight=0.2096),(Item=(Type=Bone,Name="spine_05"),Weight=0.7793))
    Weight=0.5
    ExecutePin=@RigVMFunction_Sequence.A
   )
   > &RigUnit_ParentConstraint(
    AdvancedSettings=(RotationOrderForFilter=XZY)
    Child=(Type=Bone,Name="dyn_hairB_C_03")
    Parents=((Item=(Type=Bone,Name="clavicle_l"),Weight=0.0055),(Item=(Type=Bone,Name="clavicle_r"),Weight=0.0055),(Item=(Type=Bone,Name="spine_05"),Weight=0.9891))
    Weight=1
   )
   > &RigUnit_ParentConstraint(
    AdvancedSettings=(RotationOrderForFilter=XZY)
    Child=(Type=Bone,Name="dyn_hairB_C_04")
    Parents="((Item=(Type=Bone,Name=\"deform_upperarm_up_r\"),Weight=0.0003),(Item=(Type=Bone,Name=\"upperarm_twist_01_r\"),Weight=0.0004),(Item=(Type=Bone,Name=\"deform_upperarm_bk_r\"),Weight=0.0006),(Item=(Type=Bone,Name=\"spine_04\"),Weight=0.0124),(Item=(Type=Bone,Name=\"c ...(truncated)"
    Weight=1
   )
   > &RigUnit_ParentConstraint(
    AdvancedSettings=(RotationOrderForFilter=XZY)
    Child=(Type=Bone,Name="dyn_hairB_L_02")
    Parents=((Item=(Type=Bone,Name="neck_02"),Weight=0.0003),(Item=(Type=Bone,Name="spine_03"),Weight=0.0057),(Item=(Type=Bone,Name="clavicle_l"),Weight=0.0985),(Item=(Type=Bone,Name="neck_01"),Weight=0.188),(Item=(Type=Bone,Name="spine_05"),Weight=0.7075))
    Weight=0.5
   )
   > &RigUnit_ParentConstraint(
    AdvancedSettings=(RotationOrderForFilter=XZY)
    Child=(Type=Bone,Name="dyn_hairB_L_03")
    Parents="((Item=(Type=Bone,Name=\"clavicle_r\"),Weight=0.0003),(Item=(Type=Bone,Name=\"spine_03\"),Weight=0.0004),(Item=(Type=Bone,Name=\"neck_01\"),Weight=0.0004),(Item=(Type=Bone,Name=\"spine_04\"),Weight=0.0256),(Item=(Type=Bone,Name=\"spine_05\"),Weight=0.4487),(Item=(Ty ...(truncated)"
    Weight=1
   )
   > &RigUnit_ParentConstraint(
    AdvancedSettings=(RotationOrderForFilter=XZY)
    Child=(Type=Bone,Name="dyn_hairB_L_04")
    Parents="((Item=(Type=Bone,Name=\"clavicle_r\"),Weight=0.0074),(Item=(Type=Bone,Name=\"deform_upperarm_up_l\"),Weight=0.0083),(Item=(Type=Bone,Name=\"deform_upperarm_bk_l\"),Weight=0.0138),(Item=(Type=Bone,Name=\"spine_04\"),Weight=0.0148),(Item=(Type=Bone,Name=\"upperarm_t ...(truncated)"
    Weight=1
   )
   > &RigUnit_ParentConstraint(
    AdvancedSettings=(RotationOrderForFilter=XZY)
    Child=(Type=Bone,Name="dyn_hairB_R_02")
    Parents=((Item=(Type=Bone,Name="neck_02"),Weight=0.0003),(Item=(Type=Bone,Name="spine_03"),Weight=0.0057),(Item=(Type=Bone,Name="clavicle_r"),Weight=0.0985),(Item=(Type=Bone,Name="neck_01"),Weight=0.188),(Item=(Type=Bone,Name="spine_05"),Weight=0.7075))
    Weight=0.5
   )
   > &RigUnit_ParentConstraint(
    AdvancedSettings=(RotationOrderForFilter=XZY)
    Child=(Type=Bone,Name="dyn_hairB_R_03")
    Parents=((Item=(Type=Bone,Name="spine_03"),Weight=0.0001),(Item=(Type=Bone,Name="neck_01"),Weight=0.0001),(Item=(Type=Bone,Name="spine_04"),Weight=0.1314),(Item=(Type=Bone,Name="clavicle_r"),Weight=0.3888),(Item=(Type=Bone,Name="spine_05"),Weight=0.4795))
    Weight=1
   )
   > &RigUnit_ParentConstraint(
    AdvancedSettings=(RotationOrderForFilter=XZY)
    Child=(Type=Bone,Name="dyn_hairB_R_04")
    Parents="((Item=(Type=Bone,Name=\"clavicle_l\"),Weight=0.0074),(Item=(Type=Bone,Name=\"deform_upperarm_up_r\"),Weight=0.0083),(Item=(Type=Bone,Name=\"deform_upperarm_bk_r\"),Weight=0.0138),(Item=(Type=Bone,Name=\"spine_04\"),Weight=0.0148),(Item=(Type=Bone,Name=\"upperarm_t ...(truncated)"
    Weight=1
   )
   n1 = &RigUnit_SphericalPoseReader(
    ActiveRegionScaleFactors=(PositiveWidth=0.3,NegativeWidth=0.3,PositiveHeight=0.6,NegativeHeight=1)
    ActiveRegionSize=0.15
    Debug=(bDrawDebug=false,bDraw2D=false,bDrawLocalAxes=true,DebugScale=25,DebugSegments=20,DebugThickness=0.25)
    DriverAxis=(X=1,Y=0,Z=0)
    DriverItem=(Type=Bone,Name="head")
    FalloffRegionScaleFactors=(PositiveWidth=0.8,NegativeWidth=1,PositiveHeight=1.3,NegativeHeight=1)
    FalloffSize=0.4
    FlipHeightScaling=false
    FlipWidthScaling=true
    OptionalParentItem=(Type=Bone,Name="spine_04")
    RotationOffset=(X=0,Y=-62,Z=-90)
    ExecutePin=@RigVMFunction_Sequence.B
   )
   > &RigUnit_ModifyTransforms(
    ItemToModify="((Item=(Type=Bone,Name=\"dyn_hairB_C_extra_02\"),Transform=(Rotation=(X=0,Y=0,Z=0,W=1),Translation=(X=0,Y=-5.2101,Z=0),Scale3D=(X=1,Y=1,Z=1))),(Item=(Type=Bone,Name=\"dyn_hairB_L_extra_02\"),Transform=(Rotation=(X=0,Y=0,Z=-0.1304,W=-0.9915),Translation=(X=1.94 ...(truncated)"
    WeightMinimum=0
    Weight=@n1.OutputParam
   )
   n2 = &RigVMFunction_AlphaInterp(
    InterpSpeedDecreasing=10
    InterpSpeedIncreasing=10
    Scale=1
    bClampResult=true
    Value=&RigVMFunction_MathBoolToFloat(Value=RigVMVariableNode() {variable=Crouching_Bool}.Value).Result
   )
   > &RigUnit_ModifyTransforms(
    ItemToModify="((Item=(Type=Bone,Name=\"dyn_hairB_C_02\"),Transform=(Rotation=(X=0,Y=0,Z=0,W=1),Translation=(X=0,Y=0,Z=0),Scale3D=(X=1,Y=1,Z=1))),(Item=(Type=Bone,Name=\"dyn_hairB_L_02\"),Transform=(Rotation=(X=0,Y=0,Z=0,W=1),Translation=(X=0,Y=0,Z=0),Scale3D=(X=1,Y=1,Z=1))) ...(truncated)"
    WeightMinimum=0
    Weight=&RigVMFunction_MathDoubleMul(B=@n2.Result, A=@n1.OutputParam).Result
   )
   > n3 = &RigUnit_SphericalPoseReader(
    ActiveRegionScaleFactors=(PositiveWidth=1,NegativeWidth=1,PositiveHeight=0.3,NegativeHeight=0.3)
    ActiveRegionSize=0.15
    Debug=(bDrawDebug=false,bDraw2D=false,bDrawLocalAxes=true,DebugScale=25,DebugSegments=20,DebugThickness=0.25)
    DriverAxis=(X=1,Y=0,Z=0)
    DriverItem=(Type=Bone,Name="head")
    FalloffRegionScaleFactors=(PositiveWidth=1,NegativeWidth=1.3,PositiveHeight=0.5,NegativeHeight=0.5)
    FalloffSize=0.3
    FlipHeightScaling=false
    FlipWidthScaling=true
    OptionalParentItem=(Type=Bone,Name="spine_05")
    RotationOffset=(X=-35,Y=-15,Z=15)
   )
   n4 = RigVMVariableNode() {variable=Pistol}
   Remap = &RigVMFunction_MathDoubleRemap(SourceMaximum=1, SourceMinimum=0, TargetMaximum=0, TargetMinimum=1, bClamp=False, Value=@n4.Value)
   > &RigUnit_ModifyTransforms(
    ItemToModify="((Item=(Type=Bone,Name=\"dyn_hairB_C_extra_02\"),Transform=(Rotation=(X=0,Y=0,Z=0,W=-1),Translation=(X=-0.6522,Y=0,Z=0),Scale3D=(X=1,Y=1,Z=1))),(Item=(Type=Bone,Name=\"dyn_hairB_L_extra_02\"),Transform=(Rotation=(X=0,Y=0.0786,Z=0,W=-0.9969),Translation=(X=-3.0 ...(truncated)"
    WeightMinimum=0
    Weight=&RigVMFunction_MathDoubleMul(B=@Remap.Result, A=@n3.OutputParam).Result
   )
   > &RigUnit_ModifyTransforms(
    ItemToModify="((Item=(Type=Bone,Name=\"dyn_hairB_C_extra_02\"),Transform=(Rotation=(X=0,Y=0,Z=-0.0916,W=0.9958),Translation=(X=3.2334,Y=7.7758,Z=0),Scale3D=(X=1,Y=1,Z=1))),(Item=(Type=Bone,Name=\"dyn_hairB_L_extra_02\"),Transform=(Rotation=(X=0,Y=0,Z=0,W=1),Translation=(X=- ...(truncated)"
    WeightMinimum=0
    Weight=&RigVMFunction_MathDoubleMul(B=@n4.Value, A=@n3.OutputParam).Result
   )
   RigVMFunction_MathBoolToFloat = &RigVMFunction_MathBoolToFloat(Value=RigVMVariableNode() {variable=Floating}.Value)
   AlphaInterp = &RigVMFunction_AlphaInterp(InterpSpeedDecreasing=10, InterpSpeedIncreasing=10, Scale=1, bClampResult=true, Value=@RigVMFunction_MathBoolToFloat.Result)
   > &RigUnit_ModifyTransforms(
    ItemToModify="((Item=(Type=Bone,Name=\"dyn_hairB_L_03\"),Transform=(Rotation=(X=0,Y=0,Z=-0.0842,W=-0.9964),Translation=(X=3.2253,Y=1.8828,Z=-0.2188),Scale3D=(X=1,Y=1,Z=1))),(Item=(Type=Bone,Name=\"dyn_hairB_L_02\"),Transform=(Rotation=(X=0,Y=0,Z=-0.127,W=0.9919),Translation ...(truncated)"
    WeightMinimum=0
    Weight=@AlphaInterp.Result
   )
   > &RigUnit_ModifyTransforms(
    ItemToModify="((Item=(Type=Bone,Name=\"dyn_hairB_C_extra_02\"),Transform=(Rotation=(X=0,Y=0,Z=0,W=1),Translation=(X=0,Y=0,Z=0),Scale3D=(X=1,Y=1,Z=1))),(Item=(Type=Bone,Name=\"dyn_hairB_L_extra_02\"),Transform=(Rotation=(X=0,Y=0,Z=0,W=1),Translation=(X=0,Y=0,Z=0),Scale3D=(X= ...(truncated)"
    WeightMinimum=0
    Weight=RigVMVariableNode() {variable=Glider}.Value
   )
   > Delta Rotation Y = RigVMFunctionReferenceNode(Child=clavicle_r, Parent=spine_05, SourceMax=-50, SourceMin=-80)
   Multiply = &RigVMFunction_MathDoubleMul(A=@"Delta Rotation Y".Delta Y, B=RigVMVariableNode() {variable=FallingSwim}.Value)
   n5 = &RigVMFunction_AlphaInterp(InterpSpeedDecreasing=10, InterpSpeedIncreasing=10, Scale=1, bClampResult=true, Value=@Multiply.Result)
   > &RigUnit_ModifyTransforms(
    ItemToModify="((Item=(Type=Bone,Name=\"dyn_hairB_R_02\"),Transform=(Rotation=(X=0,Y=0.1307,Z=0,W=-0.9914),Translation=(X=0,Y=0,Z=1.4317),Scale3D=(X=1,Y=1,Z=1))),(Item=(Type=Bone,Name=\"base_sd_extra_l\"),Transform=(Rotation=(X=0,Y=0.115,Z=0,W=-0.9934),Translation=(X=0,Y=0,Z ...(truncated)"
    WeightMinimum=0
    Weight=@n5.Result
   )
   n6 = &RigVMFunction_AlphaInterp(
    InterpSpeedDecreasing=10
    InterpSpeedIncreasing=10
    Scale=1
    bClampResult=true
    Value=RigVMVariableNode() {variable=CelebrationEmoteNoRban}.Value
   )
   > &RigUnit_ModifyTransforms(
    ItemToModify="((Item=(Type=Bone,Name=\"base_hairB_C_01\"),Transform=(Rotation=(X=0,Y=0,Z=0,W=1),Translation=(X=0,Y=0,Z=0),Scale3D=(X=1,Y=1,Z=1))),(Item=(Type=Bone,Name=\"base_hairB_L_01\"),Transform=(Rotation=(X=0,Y=0,Z=0,W=-1),Translation=(X=-1.8098,Y=0,Z=0),Scale3D=(X=1,Y ...(truncated)"
    WeightMinimum=0
    Weight=@n6.Result
   )
   RigVMFunctionReturnNode(ExecuteContext=@RigVMFunction_Sequence.C)
  ]}
 New Function_ContainedGraph {nodes [
   n1 = RigVMFunctionEntryNode()
   RigVMFunction_Sequence = &RigVMFunction_Sequence(ExecuteContext=@n1.ExecuteContext)
   n2 = &RigVMFunction_AlphaInterp(
    InterpSpeedDecreasing=10
    InterpSpeedIncreasing=10
    Scale=1
    bClampResult=true
    Value=&RigVMFunction_MathBoolToFloat(Value=RigVMVariableNode() {variable=Crouching_Bool}.Value).Result
   )
   RigVMVariableNode(Value=@n2.Result, ExecuteContext=@RigVMFunction_Sequence.A) {variable=Crouching}
   > n3 = &RigUnit_SphericalPoseReader(
    ActiveRegionScaleFactors=(PositiveWidth=1,NegativeWidth=1,PositiveHeight=0.3,NegativeHeight=0.3)
    ActiveRegionSize=0.15
    Debug=(bDrawDebug=false,bDraw2D=false,bDrawLocalAxes=true,DebugScale=25,DebugSegments=20,DebugThickness=0.25)
    DriverAxis=(X=1,Y=0,Z=0)
    DriverItem=(Type=Bone,Name="head")
    FalloffRegionScaleFactors=(PositiveWidth=1,NegativeWidth=1.3,PositiveHeight=0.5,NegativeHeight=0.5)
    FalloffSize=0.3
    FlipHeightScaling=false
    FlipWidthScaling=true
    OptionalParentItem=(Type=Bone,Name="spine_05")
    RotationOffset=(X=-35,Y=-15,Z=15)
   )
   n4 = RigVMVariableNode() {variable=Pistol}
   > &RigUnit_ModifyTransforms(
    ItemToModify="((Item=(Type=Bone,Name=\"base_hairB_R_01\"),Transform=(Rotation=(X=-0.2152,Y=0.1882,Z=-0.2152,W=0.9338),Translation=(X=-2.365,Y=1.3654,Z=0.7317),Scale3D=(X=1,Y=1,Z=1))),(Item=(Type=Bone,Name=\"base_sd_extra_l\"),Transform=(Rotation=(X=-0.0222,Y=-0.0939,Z=0.396 ...(truncated)"
    WeightMinimum=0
    Weight=&RigVMFunction_MathDoubleMul(
     B=&RigVMFunction_MathDoubleRemap(SourceMaximum=1, SourceMinimum=0, TargetMaximum=0, TargetMinimum=1, bClamp=False, Value=@n4.Value).Result
     A=@n3.OutputParam
    ).Result
   )
   > &RigUnit_ModifyTransforms(
    ItemToModify="((Item=(Type=Bone,Name=\"base_hairB_R_01\"),Transform=(Rotation=(X=-0.2152,Y=0.1882,Z=-0.2152,W=0.9338),Translation=(X=-1.2108,Y=-3.0735,Z=1.822),Scale3D=(X=1,Y=1,Z=1))),(Item=(Type=Bone,Name=\"base_sd_extra_l\"),Transform=(Rotation=(X=0,Y=0,Z=0,W=1),Translati ...(truncated)"
    WeightMinimum=0
    Weight=&RigVMFunction_MathDoubleMul(B=@n4.Value, A=@n3.OutputParam).Result
   )
   n5 = &RigVMFunction_AlphaInterp(
    InterpSpeedDecreasing=10
    InterpSpeedIncreasing=10
    Scale=1
    bClampResult=true
    Value=&RigVMFunction_MathBoolToFloat(Value=RigVMVariableNode() {variable=Floating}.Value).Result
   )
   > &RigUnit_ModifyTransforms(
    ItemToModify="((Item=(Type=Bone,Name=\"base_hat\"),Transform=(Rotation=(X=0,Y=0,Z=0.0484,W=-0.9988),Translation=(X=-0.9365,Y=0.3539,Z=0),Scale3D=(X=1,Y=1,Z=1))),(Item=(Type=Bone,Name=\"base_hat_rot\"),Transform=(Rotation=(X=-0.0151,Y=0.0889,Z=0,W=0.9959),Translation=(X=0,Y= ...(truncated)"
    WeightMinimum=0
    Weight=@n5.Result
   )
   > n6 = &RigUnit_SphericalPoseReader(
    ActiveRegionScaleFactors=(PositiveWidth=0.3,NegativeWidth=0.3,PositiveHeight=0.6,NegativeHeight=1)
    ActiveRegionSize=0.15
    Debug=(bDrawDebug=false,bDraw2D=false,bDrawLocalAxes=true,DebugScale=25,DebugSegments=20,DebugThickness=0.25)
    DriverAxis=(X=1,Y=0,Z=0)
    DriverItem=(Type=Bone,Name="head")
    FalloffRegionScaleFactors=(PositiveWidth=0.8,NegativeWidth=1,PositiveHeight=1.3,NegativeHeight=1)
    FalloffSize=0.4
    FlipHeightScaling=false
    FlipWidthScaling=true
    OptionalParentItem=(Type=Bone,Name="spine_04")
    RotationOffset=(X=0,Y=-62,Z=-90)
   )
   > &RigUnit_ModifyTransforms(
    ItemToModify="((Item=(Type=Bone,Name=\"base_hairB_C_01\"),Transform=(Rotation=(X=0,Y=0,Z=0,W=1),Translation=(X=0,Y=0,Z=0),Scale3D=(X=1,Y=1,Z=1))),(Item=(Type=Bone,Name=\"base_hairB_R_01\"),Transform=(Rotation=(X=0,Y=0,Z=0,W=1),Translation=(X=0,Y=0,Z=0),Scale3D=(X=1,Y=1,Z=1) ...(truncated)"
    WeightMinimum=0
    Weight=@n6.OutputParam
   )
   n7 = &RigVMFunction_AlphaInterp(
    InterpSpeedDecreasing=10
    InterpSpeedIncreasing=10
    Scale=1
    bClampResult=true
    Value=&RigVMFunction_MathBoolToFloat(Value=RigVMVariableNode() {variable=Crouching_Bool}.Value).Result
   )
   > &RigUnit_ModifyTransforms(
    ItemToModify="((Item=(Type=Bone,Name=\"base_hairB_C_01\"),Transform=(Rotation=(X=-0.1155,Y=-0.2316,Z=-0.431,W=0.8644),Translation=(X=0,Y=0,Z=0),Scale3D=(X=1,Y=1,Z=1))),(Item=(Type=Bone,Name=\"base_hairB_L_01\"),Transform=(Rotation=(X=0.0734,Y=0.1574,Z=-0.4162,W=0.8925),Tran ...(truncated)"
    WeightMinimum=0
    Weight=&RigVMFunction_MathDoubleMul(B=@n7.Result, A=@n6.OutputParam).Result
   )
   > n8 = &RigUnit_SphericalPoseReader(
    ActiveRegionScaleFactors=(PositiveWidth=3,NegativeWidth=4.5,PositiveHeight=3,NegativeHeight=3)
    ActiveRegionSize=0.08
    Debug=(bDrawDebug=false,DebugScale=25,DebugSegments=20,DebugThickness=0.25)
    DriverAxis=(X=-1,Y=0,Z=0)
    DriverItem=(Type=Bone,Name="upperarm_r")
    FalloffRegionScaleFactors=(PositiveWidth=2.1,NegativeWidth=5,PositiveHeight=2,NegativeHeight=2)
    OptionalParentItem=(Type=Bone,Name="spine_04")
    RotationOffset=(X=0,Y=180,Z=0)
    FalloffSize=&RigVMFunction_MathDoubleRemap(SourceMaximum=1, SourceMinimum=0, TargetMaximum=0.43, TargetMinimum=0.41, bClamp=False, Value=RigVMVariableNode() {variable=Crouching}.Value).Result
   )
   n9 = &RigVMFunction_AlphaInterp(
    InterpSpeedDecreasing=10
    InterpSpeedIncreasing=10
    Scale=1
    bClampResult=true
    Value=&RigVMFunction_MathBoolToFloat(Value=RigVMVariableNode() {variable=Skydiving}.Value).Result
   )
   Multiply_6 = RigVMAggregateNode(
    D=&RigVMFunction_MathDoubleRemap(SourceMaximum=1, SourceMinimum=0, TargetMaximum=0, TargetMinimum=1, Value=@n9.Result).Result
    C=&RigVMFunction_MathDoubleRemap(SourceMaximum=1, SourceMinimum=0, TargetMaximum=0, TargetMinimum=1, Value=RigVMVariableNode() {variable=Crouching}.Value).Result
    A=@n8.OutputParam
    B=RigVMVariableNode() {variable=Pistol}.Value
   )
   > &RigUnit_ModifyTransforms(
    ItemToModify="((Item=(Type=Bone,Name=\"base_sd_extra_l\"),Transform=(Rotation=(X=-0.2673,Y=0.3867,Z=-0.2925,W=0.8327),Translation=(X=-0.5843,Y=3.0942,Z=-0.7494),Scale3D=(X=1,Y=1,Z=1))),(Item=(Type=Bone,Name=\"base_hairB_R_01\"),Transform=(Rotation=(X=-0.127,Y=0.1405,Z=0.145 ...(truncated)"
    WeightMinimum=0
    Weight=&RigVMFunction_MathDoubleRemap(SourceMaximum=0.65, SourceMinimum=0, TargetMaximum=1, TargetMinimum=0, Value=@Multiply_6.Result).Result
   )
   Multiply_13 = RigVMAggregateNode(A=RigVMVariableNode() {variable=Pistol}.Value, C=RigVMVariableNode() {variable=Crouching}.Value, B=@n8.OutputParam)
   > &RigUnit_ModifyTransforms(
    ItemToModify="((Item=(Type=Bone,Name=\"base_sd_extra_l\"),Transform=(Rotation=(X=0.115,Y=0.1391,Z=0.5647,W=-0.8053),Translation=(X=0,Y=0,Z=0),Scale3D=(X=1,Y=1,Z=1))),(Item=(Type=Bone,Name=\"base_hairB_R_01\"),Transform=(Rotation=(X=0.1607,Y=0.3404,Z=-0.0303,W=-0.926),Transl ...(truncated)"
    WeightMinimum=0
    Weight=&RigVMFunction_MathDoubleRemap(SourceMaximum=0.3, SourceMinimum=0, TargetMaximum=1, TargetMinimum=0, Value=@Multiply_13.Result).Result
   )
   > Delta Rotation Z = RigVMFunctionReferenceNode(Child=upperarm_r, Parent=clavicle_r, SourceMax=150, SourceMin=70)
   n10 = &RigVMFunction_AlphaInterp(
    InterpSpeedDecreasing=10
    InterpSpeedIncreasing=10
    Scale=1
    bClampResult=true
    bMapRange=false
    Value=&RigVMFunction_MathBoolToFloat(Value=RigVMVariableNode() {variable=Moving2d}.Value).Result
   )
   Multiply_8 = RigVMAggregateNode(
    D=@n10.Result
    A=@"Delta Rotation Z".Delta Z
    C=RigVMVariableNode() {variable=Pistol}.Value
    B=RigVMVariableNode() {variable=Crouching}.Value
   )
   > &RigUnit_ModifyTransforms(
    ItemToModify="((Item=(Type=Bone,Name=\"base_hairB_R_01\"),Transform=(Rotation=(X=0,Y=-0.0589,Z=0,W=0.9983),Translation=(X=8.127,Y=0,Z=-0.0915),Scale3D=(X=1,Y=1,Z=1))),(Item=(Type=Bone,Name=\"base_hairB_L_01\"),Transform=(Rotation=(X=0.0689,Y=-0.0991,Z=-0.4137,W=0.9024),Tran ...(truncated)"
    WeightMinimum=0
    Weight=@Multiply_8.Result
   )
   > n11 = &RigUnit_SphericalPoseReader(
    ActiveRegionScaleFactors=(PositiveWidth=3,NegativeWidth=3,PositiveHeight=3,NegativeHeight=3)
    ActiveRegionSize=0.08
    Debug=(bDrawDebug=false,DebugScale=25,DebugSegments=20,DebugThickness=0.25)
    DriverAxis=(X=1,Y=0,Z=0)
    DriverItem=(Type=Bone,Name="upperarm_l")
    FalloffRegionScaleFactors=(PositiveWidth=2.1,NegativeWidth=2,PositiveHeight=2,NegativeHeight=2)
    FalloffSize=0.37
    OptionalParentItem=(Type=Bone,Name="spine_04")
    RotationOffset=(X=-20,Y=0,Z=0)
   )
   Multiply_15 = RigVMAggregateNode(
    B=&RigVMFunction_MathDoubleRemap(SourceMaximum=1, SourceMinimum=0, TargetMaximum=0, TargetMinimum=1, Value=RigVMVariableNode() {variable=Pistol}.Value).Result
    A=@n11.OutputParam
    C=RigVMVariableNode() {variable=Crouching}.Value
   )
   > &RigUnit_ModifyTransforms(
    ItemToModify="((Item=(Type=Bone,Name=\"base_hair_sd_d_extra_r\"),Transform=(Rotation=(X=0,Y=0,Z=0,W=-1),Translation=(X=-2.2396,Y=0.4187,Z=1.8975),Scale3D=(X=1,Y=1,Z=1))),(Item=(Type=Bone,Name=\"base_hairB_R_01\"),Transform=(Rotation=(X=-0.1001,Y=0.1586,Z=-0.1115,W=-0.9759), ...(truncated)"
    WeightMinimum=0
    Weight=@Multiply_15.Result
   )
   > n12 = &RigUnit_SphericalPoseReader(
    ActiveRegionScaleFactors=(PositiveWidth=0.75,NegativeWidth=0.75,PositiveHeight=1,NegativeHeight=1)
    ActiveRegionSize=0.1
    Debug=(bDrawDebug=false,DebugScale=25,DebugSegments=20,DebugThickness=0.25)
    DriverAxis=(X=0,Y=1,Z=0)
    DriverItem=(Type=Bone,Name="head")
    FalloffRegionScaleFactors=(PositiveWidth=1,NegativeWidth=1,PositiveHeight=1,NegativeHeight=2.2)
    FalloffSize=0.2
    OptionalParentItem=(Type=Bone)
    RotationOffset=(X=150,Y=0,Z=0)
   )
   > &RigUnit_ModifyTransforms(
    ItemToModify="((Item=(Type=Bone,Name=\"base_hairB_L_01\"),Transform=(Rotation=(X=-0.0777,Y=-0.1681,Z=-0.0207,W=0.9825),Translation=(X=0,Y=0,Z=0),Scale3D=(X=1,Y=1,Z=1))),(Item=(Type=Bone,Name=\"base_hairB_C_01\"),Transform=(Rotation=(X=-0.2236,Y=-0.3074,Z=-0.0238,W=0.9246),T ...(truncated)"
    WeightMinimum=0
    Weight=@n12.OutputParam
   )
   > n13 = &RigUnit_SphericalPoseReader(
    ActiveRegionScaleFactors=(PositiveWidth=1,NegativeWidth=1,PositiveHeight=1,NegativeHeight=1)
    ActiveRegionSize=0.15
    Debug=(bDrawDebug=false,bDraw2D=false,bDrawLocalAxes=true,DebugScale=25,DebugSegments=20,DebugThickness=0.25)
    DriverAxis=(X=1,Y=0,Z=0)
    DriverItem=(Type=Bone,Name="neck_01")
    FalloffRegionScaleFactors=(PositiveWidth=1,NegativeWidth=1,PositiveHeight=1,NegativeHeight=1)
    FalloffSize=0.4
    FlipHeightScaling=false
    FlipWidthScaling=true
    OptionalParentItem=(Type=Bone,Name="spine_04")
    RotationOffset=(X=-90,Y=90,Z=150)
   )
   > &RigUnit_ModifyTransforms(
    ItemToModify="((Item=(Type=Bone,Name=\"base_hairB_L_01\"),Transform=(Rotation=(X=0,Y=0,Z=0.2528,W=-0.9675),Translation=(X=0,Y=0,Z=0),Scale3D=(X=1,Y=1,Z=1))),(Item=(Type=Bone,Name=\"base_sd_extra_l\"),Transform=(Rotation=(X=0,Y=0,Z=-0.1358,W=0.9907),Translation=(X=-0.395,Y=3 ...(truncated)"
    WeightMinimum=0
    Weight=@n13.OutputParam
   )
   n14 = &RigVMFunction_AlphaInterp(
    InterpSpeedDecreasing=10
    InterpSpeedIncreasing=10
    Scale=1
    bClampResult=true
    Value=RigVMVariableNode() {variable=CelebrationEmoteRban}.Value
   )
   > &RigUnit_ModifyTransforms(
    ItemToModify="((Item=(Type=Bone,Name=\"base_hairB_L_01\"),Transform=(Rotation=(X=0,Y=0,Z=-0.0262,W=0.9997),Translation=(X=-3.5,Y=0,Z=0),Scale3D=(X=1,Y=1,Z=1))),(Item=(Type=Bone,Name=\"base_hairB_R_01\"),Transform=(Rotation=(X=0,Y=0,Z=-0.0262,W=-0.9997),Translation=(X=-0.955 ...(truncated)"
    WeightMinimum=0
    Weight=@n14.Result
   )
   > &RigUnit_ModifyTransforms(
    ItemToModify="((Item=(Type=Bone,Name=\"base_hairB_C_01\"),Transform=(Rotation=(X=0,Y=0,Z=0,W=-1),Translation=(X=0,Y=-1.8416,Z=0),Scale3D=(X=1,Y=1,Z=1))),(Item=(Type=Bone,Name=\"base_hairB_R_01\"),Transform=(Rotation=(X=0,Y=0,Z=0,W=-1),Translation=(X=0,Y=-0.9684,Z=0),Scale3D ...(truncated)"
    WeightMinimum=0
    Weight=RigVMVariableNode() {variable=Glider}.Value
   )
   RigVMFunction_MathBoolToFloat = &RigVMFunction_MathBoolToFloat(Value=RigVMVariableNode() {variable=Floating}.Value)
   AlphaInterp = &RigVMFunction_AlphaInterp(InterpSpeedDecreasing=10, InterpSpeedIncreasing=10, Scale=1, bClampResult=true, Value=@RigVMFunction_MathBoolToFloat.Result)
   > &RigUnit_ModifyTransforms(
    ItemToModify="((Item=(Type=Bone,Name=\"base_hairB_C_01\"),Transform=(Rotation=(X=0,Y=0,Z=0.0872,W=-0.9962),Translation=(X=-0.4591,Y=-2.6034,Z=0),Scale3D=(X=1,Y=1,Z=1))),(Item=(Type=Bone,Name=\"base_hairB_R_01\"),Transform=(Rotation=(X=0,Y=0,Z=0.0872,W=-0.9962),Translation=( ...(truncated)"
    WeightMinimum=0
    Weight=@AlphaInterp.Result
   )
   > Delta Rotation Y = RigVMFunctionReferenceNode(Child=clavicle_r, Parent=spine_05, SourceMax=-50, SourceMin=-80)
   Multiply = &RigVMFunction_MathDoubleMul(A=@"Delta Rotation Y".Delta Y, B=RigVMVariableNode() {variable=FallingSwim}.Value)
   n15 = &RigVMFunction_AlphaInterp(InterpSpeedDecreasing=10, InterpSpeedIncreasing=10, Scale=1, bClampResult=true, Value=@Multiply.Result)
   > n16 = &RigUnit_ModifyTransforms(
    ItemToModify="((Item=(Type=Bone,Name=\"base_hairB_R_01\"),Transform=(Rotation=(X=0,Y=-0.0652,Z=0,W=0.9979),Translation=(X=-0.4101,Y=0,Z=2.3013),Scale3D=(X=1,Y=1,Z=1))),(Item=(Type=Bone,Name=\"base_sd_extra_l\"),Transform=(Rotation=(X=0,Y=-0.115,Z=0,W=0.9934),Translation=(X= ...(truncated)"
    WeightMinimum=0
    Weight=@n15.Result
   )
   RigVMFunction_ControlFlowBranch = &RigVMFunction_ControlFlowBranch(ExecuteContext=@n16.ExecutePin, Condition=RigVMVariableNode() {variable=playEmote}.Value)
   Delta Rotation Y_1 = RigVMFunctionReferenceNode(Child=clavicle_r, Parent=spine_05, SourceMax=-64, SourceMin=-72, ExecuteContext=@RigVMFunction_ControlFlowBranch.True)
   > &RigUnit_ModifyTransforms(
    ItemToModify=((Item=(Type=Bone,Name="dyn_hair_sd_f_01_r"),Transform=(Rotation=(X=0.0862,Y=-0.0085,Z=0.0974,W=-0.9915),Translation=(X=-0.7347,Y=-2.8343,Z=1.4523),Scale3D=(X=1,Y=1,Z=1))))
    WeightMinimum=0
    Weight=@"Delta Rotation Y_1".Delta Y
   )
   RigVMFunctionReturnNode(ExecuteContext=@RigVMFunction_Sequence.B)
  ]}
 New Function_ContainedGraph {nodes [
   n1 = RigVMFunctionEntryNode()
   RigVMFunction_MathQuaternionToEuler = &RigVMFunction_MathQuaternionToEuler(
    RotationOrder=ZXY
    Value=&RigUnit_GetRelativeTransformForItem(
     Child=(Type=Bone,Name="None")
     Parent=(Type=Bone,Name="None")
     bChildInitial=False
     bParentInitial=False
     Child.Name=@n1.Child
     Parent.Name=@n1.Parent
    ).RelativeTransform.Rotation
   )
   Remap = &RigVMFunction_MathDoubleRemap(
    TargetMaximum=1
    TargetMinimum=0
    Value=@RigVMFunction_MathQuaternionToEuler.Result.Z
    SourceMinimum=@n1.SourceMin
    SourceMaximum=@n1.SourceMax
   )
   > RigVMFunctionReturnNode(Delta Z=@Remap.Result)
  ]}
 RigVMFunctionLibrary {nodes [
   Constrains Backpack = RigVMCollapseNode(ExecuteContext=())
   Delta Rotation Y = RigVMCollapseNode(ExecuteContext=())
   Delta Rotation Z = RigVMCollapseNode(ExecuteContext=())
   Setup Hair = RigVMCollapseNode(ExecuteContext=())
  ]}
 RigVMFunction_Sequence_1_ContainedGraph {nodes [
   n1 = RigVMFunctionEntryNode()
   RigVMFunction_Sequence = &RigVMFunction_Sequence(ExecuteContext=@n1.ExecuteContext)
   n2 = &RigVMFunction_Sequence(ExecuteContext=@RigVMFunction_Sequence.B)
   RigVMFunctionReturnNode(A=@RigVMFunction_Sequence.A, B=@n2.A, C=@n2.B)
  ]}
 RigVMModel {nodes [
   RigUnit_BeginExecution = &RigUnit_BeginExecution()
   RigVMFunction_ControlFlowBranch = &RigVMFunction_ControlFlowBranch(ExecuteContext=@RigUnit_BeginExecution.ExecutePin, Condition=RigVMVariableNode() {variable=Wearing BackblingAndCape}.Value)
   Setup Hair = RigVMFunctionReferenceNode(ExecuteContext=@RigVMFunction_ControlFlowBranch.False)
   Constrains Backpack = RigVMFunctionReferenceNode(ExecuteContext=@RigVMFunction_ControlFlowBranch.True)
   &RigUnit_PrepareForExecution()
   addFKCtrl = RigVMFunctionReferenceNode(
    Custom_FKCtrl=()
    ExecuteContext=()
    Options_FKCtrl="(pelvis_1_#1=False,spine_3_#2=False,clavicle_5_#3=true,arm_7_#4=False,leg_9_#5=False,head_11_46DC15EF465CD ...(truncated)"
    Shape_Settings="(pelvis_2_#6=(bVisible=True,Name=\"Hexagon_Thin\",Color=(R=0.0117,G=0.8281,B=1,A=1),Transform=(Rotation=(X=0,Y=-0.7071,Z=0,W=0.7071),Translation=(X=2,Y=0,Z=0),Scale3D=(X=5,Y=5,Z=7))),spine_4_#7=(bVi ...(truncated)"
   )
   setFKCtrl = RigVMFunctionReferenceNode(ExecuteContext=(), setFK_Ctrl=false)
  ]}
]
hierarchy [
 root {loc=<0 0 0>, rot=<0 0 0 1>}
 pelvis {parent=root, loc=<0 -0.5301 98.0571>, rot=<0.0127 0.707 -0.0127 -0.707>}
 spine_01 {parent=pelvis, loc=<6.1208 -0.9328 0>, rot=<0 0 0.034 -0.9994>}
 spine_02 {parent=spine_01, loc=<6.5997 -0.2095 0>, rot=<0 0 0.0238 -0.9997>}
 spine_03 {parent=spine_02, loc=<6.4914 -0.1237 0>, rot=<0 0 0.0149 -0.9999>}
 spine_04 {parent=spine_03, loc=<6.9598 0.0794 0>, rot=<0 0 -0.0132 -0.9999>}
 spine_05 {parent=spine_04, loc=<6.9987 0 0>, rot=<0 0 -0.0015 -1>}
 neck_01 {parent=spine_05, loc=<11.3138 0.1529 0>, rot=<0 0 -0.2003 -0.9797>}
 neck_02 {parent=neck_01, loc=<4.5627 0 0>, rot=<0 0 0.0504 -0.9987>}
 head {parent=neck_02, loc=<5.3914 0 0>, rot=<0 0 0.115 -0.9934>}
 base_hairB_C_01 {parent=head, loc=<7.1957 -8.1033 0.3833>, rot=<0.061 -0.9981 0 0>}
 dyn_hairB_C_01 {parent=base_hairB_C_01, loc=<0 0 0>, rot=<0 0 0 1>}
 dyn_hairB_C_02 {parent=dyn_hairB_C_01, loc=<8.2212 0 0>, rot=<0 0 0.0828 -0.9966>}
 dyn_hairB_C_03 {parent=dyn_hairB_C_02, loc=<8.2594 0 0>, rot=<0 0 0.0338 -0.9994>}
 dyn_hairB_C_04 {parent=dyn_hairB_C_03, loc=<8.4746 0 0>, rot=<0 0 -0.1226 -0.9925>}
 base_hair_bk_a_r {parent=dyn_hairB_C_02, loc=<2.0048 -1.7562 -3.0195>, rot=<-0.2011 0.3602 0.2074 -0.887>}
 dyn_hair_bk_a_01_r {parent=base_hair_bk_a_r, loc=<2.4965 0 0>, rot=<0 0 0 1>}
 base_hair_bk_c_l {parent=dyn_hairB_C_02, loc=<5.2943 -2.0797 2.6263>, rot=<-0.1012 0.1952 0.101 -0.9703>}
 dyn_hair_bk_c_01_l {parent=base_hair_bk_c_l, loc=<-0.0078 0 0>, rot=<0 0 0 1>}
 dyn_hair_bk_c_02_l {parent=dyn_hair_bk_c_01_l, loc=<2.046 0 0>, rot=<0 -0.2753 0 -0.9614>}
 dyn_hair_bk_c_03_l {parent=dyn_hair_bk_c_02_l, loc=<2.6039 0 0>, rot=<0 0 0 1>}
 dyn_hairB_C_extra_02 {parent=dyn_hairB_C_02, loc=<0 0 0>, rot=<0 0 0 1>}
 base_hairB_L_01 {parent=head, loc=<7.3192 -5.295 -5.9044>, rot=<0.0956 -0.9954 0 0>}
 dyn_hairB_L_01 {parent=base_hairB_L_01, loc=<0 0 0>, rot=<0 0 0 1>}
 dyn_hairB_L_02 {parent=dyn_hairB_L_01, loc=<8.2168 0 0>, rot=<0 0 0.0901 -0.9959>}
 dyn_hairB_L_03 {parent=dyn_hairB_L_02, loc=<8.4053 0 0.5102>, rot=<0 0 0.0288 -0.9996>}
 dyn_hairB_L_04 {parent=dyn_hairB_L_03, loc=<7.1171 0 0>, rot=<0 0 -0.1625 -0.9867>}
 base_hair_bk_b_l {parent=dyn_hairB_L_02, loc=<4.5168 -2.6513 -0.3129>, rot=<0.0304 -0.0489 0.1905 -0.98>}
 dyn_hair_bk_b_01_l {parent=base_hair_bk_b_l, loc=<0 0 0>, rot=<0 0 0 1>}
 dyn_hair_bk_b_02_l {parent=dyn_hair_bk_b_01_l, loc=<2.1674 0 0>, rot=<0 0 0 1>}
 base_hair_sd_e_l {parent=dyn_hairB_L_02, loc=<2.2411 2.4309 4.3472>, rot=<-0.0153 0.0986 -0.3247 -0.9406>}
 dyn_hair_sd_e_01_l {parent=base_hair_sd_e_l, loc=<0 0 0>, rot=<0 0 0 1>}
 dyn_hair_sd_e_02_l {parent=dyn_hair_sd_e_01_l, loc=<3.3867 0 0>, rot=<0 -0.471 0 -0.8821>}
 dyn_hair_sd_e_03_l {parent=dyn_hair_sd_e_02_l, loc=<2.5204 0 0>, rot=<0 0 0 1>}
 base_hair_sd_d_l {parent=dyn_hairB_L_02, loc=<1.3142 5.0385 2.942>, rot=<0.8995 0.1211 -0.1437 -0.3945>}
 dyn_hair_sd_d_01_l {parent=base_hair_sd_d_l, loc=<0 0 0>, rot=<0 0 0 1>}
 dyn_hair_sd_d_02_l {parent=dyn_hair_sd_d_01_l, loc=<2.7923 0 0>, rot=<0 -0.0651 -0.1797 -0.9816>}
 dyn_hair_sd_d_03_l {parent=dyn_hair_sd_d_02_l, loc=<2.4503 0 0>, rot=<0 0 0 1>}
 dyn_hairB_L_extra_02 {parent=dyn_hairB_L_02, loc=<0 0 0>, rot=<0 0 0 1>}
 base_hair_bk_a_l {parent=dyn_hairB_L_01, loc=<4.6118 -2.3063 0.5805>, rot=<0.5498 0.1664 0.2598 -0.7762>}
 dyn_hair_bk_a_01_l {parent=base_hair_bk_a_l, loc=<0 0 0>, rot=<0 0 0 1>}
 dyn_hair_bk_a_02_l {parent=dyn_hair_bk_a_01_l, loc=<3.6756 0 0>, rot=<0 0 0 1>}
 base_hair_sd_f_l {parent=dyn_hairB_L_01, loc=<6.8347 -0.5729 3.3255>, rot=<-0.0456 0.3667 -0.0552 -0.9276>}
 base_hair_sd_f_01_l {parent=base_hair_sd_f_l, loc=<0 0 0>, rot=<0.1059 0 0 -0.9944>}
 base_hair_sd_f_02_l {parent=base_hair_sd_f_01_l, loc=<3.6627 0 0>, rot=<0 0 0 1>}
 base_hairB_R_01 {parent=head, loc=<7.1722 -5.5783 6.1865>, rot=<0.0805 -0.9968 0 0>}
 dyn_hairB_R_01 {parent=base_hairB_R_01, loc=<0 0 0>, rot=<0 0 0 -1>}
 dyn_hairB_R_02 {parent=dyn_hairB_R_01, loc=<8.2749 0.6765 -0.0053>, rot=<0 0 0.0734 -0.9973>}
 dyn_hairB_R_03 {parent=dyn_hairB_R_02, loc=<8.1762 -0.6758 -0.64>, rot=<0 0 0.0344 -0.9994>}
 dyn_hairB_R_04 {parent=dyn_hairB_R_03, loc=<7.1579 0 -0.3731>, rot=<0 0 -0.0869 -0.9962>}
 base_hair_sd_e_r {parent=dyn_hairB_R_02, loc=<3.7065 -0.9462 -3.2149>, rot=<0.2332 -0.2994 0.3275 -0.8653>}
 base_hair_sd_e_01_r {parent=base_hair_sd_e_r, loc=<0 0 0>, rot=<0 0 0 1>}
 base_hair_sd_e_02_r {parent=base_hair_sd_e_01_r, loc=<2.4321 0 0>, rot=<0 0 0 1>}
 base_hair_sd_f_r {parent=dyn_hairB_R_02, loc=<3.7671 2.125 -5.6003>, rot=<-0.2884 -0.1048 0.3293 -0.893>}
 dyn_hair_sd_f_01_r {parent=base_hair_sd_f_r, loc=<0 0 0>, rot=<0 0 0 1>}
 dyn_hair_sd_f_02_r {parent=dyn_hair_sd_f_01_r, loc=<3.0777 0 0>, rot=<0 0.278 0 -0.9606>}
 dyn_hair_sd_f_03_r {parent=dyn_hair_sd_f_02_r, loc=<3.1555 0 0>, rot=<0 0 0 1>}
 base_hair_sd_g_r {parent=dyn_hair_sd_f_01_r, loc=<3.9233 2.9931 0.4857>, rot=<0.191 0.2215 -0.3002 -0.9079>}
 base_hair_sd_g_01_r {parent=base_hair_sd_g_r, loc=<0 0 0>, rot=<0 0 0 1>}
 base_hair_sd_g_02_r {parent=base_hair_sd_g_01_r, loc=<2.1462 0 0>, rot=<0 0 0 1>}
 dyn_hairB_R_extra_02 {parent=dyn_hairB_R_02, loc=<0 0 0>, rot=<0 0 0 1>}
 base_hair_sd_h_r {parent=dyn_hairB_R_01, loc=<-0.6256 1.0105 -2.4402>, rot=<-0.4499 0.2877 0.3144 -0.7848>}
 dyn_hair_sd_h_01_r {parent=base_hair_sd_h_r, loc=<0 0 0>, rot=<0 0 0 1>}
 dyn_hair_sd_h_02_r {parent=dyn_hair_sd_h_01_r, loc=<4.2317 0 0>, rot=<0 0 0 1>}
 base_hair_sd_b_l {parent=head, loc=<-1.8672 0.3663 -7.8737>, rot=<-0.1491 -0.9827 -0.0838 0.0707>}
 dyn_hair_sd_b_01_l {parent=base_hair_sd_b_l, loc=<0 0 0>, rot=<0 0 0 1>}
 dyn_hair_sd_b_02_l {parent=dyn_hair_sd_b_01_l, loc=<3.7203 0 0>, rot=<0 0 0 1>}
 base_hair_sd_b_r {parent=head, loc=<0.574 4.1127 7.2673>, rot=<-0.0574 -0.5114 0.8574 -0.0083>}
 dyn_hair_sd_b_01_r {parent=base_hair_sd_b_r, loc=<0 0 0>, rot=<0 0 0 1>}
 dyn_hair_sd_b_02_r {parent=dyn_hair_sd_b_01_r, loc=<5.5212 0.5185 0.0969>, rot=<0 0 -0.0859 -0.9963>}
 base_hair_fr_a_r {parent=head, loc=<7.9144 9.8243 2.431>, rot=<0.004 -0.9787 -0.0008 -0.2053>}
 dyn_hair_fr_a_01_r {parent=base_hair_fr_a_r, loc=<0 0 0>, rot=<0 0 0 1>}
 dyn_hair_fr_a_02_r {parent=dyn_hair_fr_a_01_r, loc=<4.323 0 0>, rot=<0 0 0 1>}
 base_hair_sd_a_l {parent=head, loc=<2.7209 4.8417 -5.9855>, rot=<-0.0152 -0.8393 0.4718 -0.2698>}
 dyn_hair_sd_a_01_l {parent=base_hair_sd_a_l, loc=<0 0 0>, rot=<0.0215 0.1927 0.148 -0.9698>}
 dyn_hair_sd_a_02_l {parent=dyn_hair_sd_a_01_l, loc=<4.5638 0.1993 0>, rot=<0 0.1634 0 -0.9866>}
 dyn_hair_sd_a_03_l {parent=dyn_hair_sd_a_01_l, loc=<4.3428 0.3334 1.1341>, rot=<0 0.4044 0 -0.9146>}
 dyn_hair_sd_a_04_l {parent=dyn_hair_sd_a_01_l, loc=<4.5935 -0.0542 -1.6545>, rot=<0 -0.0214 0 -0.9998>}
 base_hair_sd_a_r {parent=head, loc=<3.8141 7.5984 6.4293>, rot=<0.0217 -0.8885 -0.4491 -0.0913>}
 dyn_hair_sd_a_01_r {parent=base_hair_sd_a_r, loc=<0 0 0>, rot=<0 0 0 1>}
 dyn_hair_sd_a_02_r {parent=dyn_hair_sd_a_01_r, loc=<2.4648 -0.1928 0.7269>, rot=<0.0404 0.0611 0.1818 -0.9806>}
 dyn_hair_sd_a_03_r {parent=dyn_hair_sd_a_01_r, loc=<2.5191 -0.1108 -1.3353>, rot=<0.1043 0.1085 0.0272 -0.9882>}
 base_hair_fr_a_l {parent=head, loc=<8.5523 9.4053 -3.2311>, rot=<0.1038 -0.8715 0.3188 -0.3579>}
 dyn_hair_fr_a_01_l {parent=base_hair_fr_a_l, loc=<0 0 0>, rot=<0 0 0 1>}
 dyn_hair_fr_a_02_l {parent=dyn_hair_fr_a_01_l, loc=<3.8181 0 0>, rot=<0 0 0 1>}
 base_hair_sd_c_l {parent=head, loc=<2.6952 -1.6169 -8.8767>, rot=<0.261 -0.7344 0.38 -0.4981>}
 dyn_hair_sd_c_01_l {parent=base_hair_sd_c_l, loc=<0 0 0>, rot=<0 0 0 1>}
 dyn_hair_sd_c_02_l {parent=dyn_hair_sd_c_01_l, loc=<3.086 0 0>, rot=<0 0 0 1>}
 base_hair_bk_b_r {parent=head, loc=<0.3554 -11.1121 3.2362>, rot=<0.1046 -0.8749 0.1915 -0.4323>}
 dyn_hair_bk_b_01_r {parent=base_hair_bk_b_r, loc=<0 0 0>, rot=<0 0 0 1>}
 base_hair_sd_j_r {parent=head, loc=<3.2598 -0.0557 10.3616>, rot=<0.4941 -0.5463 0.6124 -0.287>}
 dyn_hair_sd_j_01_r {parent=base_hair_sd_j_r, loc=<0 0 0>, rot=<0 0 0 1>}
 dyn_hair_sd_j_02_r {parent=dyn_hair_sd_j_01_r, loc=<3.7247 0 0>, rot=<0 0 0 1>}
 base_hair_sd_i_r {parent=head, loc=<8.0671 0.7063 8.687>, rot=<0.5217 -0.639 0.5384 -0.1725>}
 dyn_hair_sd_i_01_r {parent=base_hair_sd_i_r, loc=<0 0 0>, rot=<0 0 0 1>}
 dyn_hair_sd_i_02_r {parent=dyn_hair_sd_i_01_r, loc=<2.2466 0 0>, rot=<0 0 0 1>}
 base_hair_bk_c_r {parent=head, loc=<9.1423 -9.5065 3.646>, rot=<0.3298 -0.9238 0.1692 -0.0955>}
 dyn_hair_bk_c_01_r {parent=base_hair_bk_c_r, loc=<0 0 0>, rot=<0 0 0 1>}
 dyn_hair_bk_c_02_r {parent=dyn_hair_bk_c_01_r, loc=<2.1562 0 0>, rot=<0 0 0 1>}
 base_sd_extra_l {parent=head, loc=<1.7949 -0.5831 8.7054>, rot=<0.0805 -0.9968 0 0>}
 dyn_sd_extra_l {parent=base_sd_extra_l, loc=<0 0 0>, rot=<0 0 0 1>}
 base_hair_sd_d_r {parent=dyn_sd_extra_l, loc=<7.0282 0.3289 -1.6837>, rot=<0.8458 -0.1191 -0.3913 0.3424>}
 dyn_hair_sd_d_01_r {parent=base_hair_sd_d_r, loc=<0 0 0>, rot=<0 0 0 1>}
 dyn_hair_sd_d_02_r {parent=dyn_hair_sd_d_01_r, loc=<2.3632 0 0>, rot=<0 -0.0637 -0.0284 -0.9976>}
 dyn_hair_sd_d_03_r {parent=dyn_hair_sd_d_02_r, loc=<2.7592 0 0>, rot=<0 0 0 1>}
 base_hair_sd_d_extra_r {parent=base_hair_sd_d_r, loc=<0 0 0>, rot=<0 0 0 1>}
 base_hair_sd_c_r {parent=dyn_sd_extra_l, loc=<5.085 1.0995 -1.4686>, rot=<0.7595 0.1078 -0.4891 0.4151>}
 dyn_hair_sd_c_01_r {parent=base_hair_sd_c_r, loc=<0 0 0>, rot=<0 0 0 1>}
 dyn_hair_sd_c_02_r {parent=dyn_hair_sd_c_01_r, loc=<3.1769 0 0>, rot=<0 0 0 1>}
 base_hat {parent=head, loc=<7.8414 -2.203 0>, rot=<0 0 -0.0041 -1>}
 base_bk_bow_l {parent=base_hat, loc=<-0.6733 2.2594 -8.5266>, rot=<0.4722 -0.7348 0.16 -0.4598>}
 dyn_bk_bow_01_l {parent=base_bk_bow_l, loc=<0 0 0>, rot=<0 0 0 1>}
 base_hat_fr_01_m {parent=base_hat, loc=<1.4385 11.7184 1.2>, rot=<-0.5 -0.5 -0.5 -0.5>}
 dyn_hat_fr_01_m {parent=base_hat_fr_01_m, loc=<0 0 0>, rot=<0 0 0 1>}
 base_hat_rot {parent=base_hat, loc=<0 0 0>, rot=<0 0 0 1>}
 base_hat_sd_01_l {parent=base_hat_rot, loc=<-1.664 3.0618 -9.1491>, rot=<-0.106 -0.6991 0.106 -0.6991>}
 dyn_hat_sd_01_l {parent=base_hat_sd_01_l, loc=<0 0 0>, rot=<0 0 0 1>}
 base_sd_bow_l {parent=dyn_hat_sd_01_l, loc=<0.1602 0.3026 0.0213>, rot=<0.1244 -0.2271 -0.1752 -0.9499>}
 dyn_sd_bow_01_l {parent=base_sd_bow_l, loc=<0 0 0>, rot=<0 0 0 1>}
 dyn_sd_bow_02_l {parent=dyn_sd_bow_01_l, loc=<4.9628 0 0>, rot=<0 0 0 1>}
 base_hat_bk_01_m {parent=base_hat_rot, loc=<-3.7237 -8.7048 0>, rot=<0.4194 -0.5693 0.5693 -0.4194>}
 dyn_hat_bk_01_m {parent=base_hat_bk_01_m, loc=<0 0 0>, rot=<0 0 0 1>}
 base_hat_sd_01_r {parent=base_hat_rot, loc=<-1.7187 3.1007 8.8092>, rot=<0.6991 -0.106 0.6991 0.106>}
 dyn_hat_sd_01_r {parent=base_hat_sd_01_r, loc=<0 0 0>, rot=<0 0 0 1>}
 clavicle_l {parent=spine_05, loc=<5.4831 -0.1053 -3.2946>, rot=<0.0465 -0.7289 -0.009 -0.6829>}
 upperarm_l {parent=clavicle_l, loc=<9.8658 -1.2136 -0.4105>, rot=<0.0164 -0.3245 -0.0269 -0.9454>}
 upperarm_twist_01_l {parent=upperarm_l, loc=<7.2135 -0.1134 0>, rot=<0 0 0 1>}
 deform_upperarm_bk_base_l {parent=upperarm_twist_01_l, loc=<-6.4053 -4.3746 0.314>, rot=<0 0 0 -1>}
 deform_upperarm_bk_l {parent=deform_upperarm_bk_base_l, loc=<0 0 0>, rot=<0 0 0 1>}
 deform_upperarm_fr_base_l {parent=upperarm_twist_01_l, loc=<-5.6312 5.092 0.3354>, rot=<0 0 0 -1>}
 deform_upperarm_fr_l {parent=deform_upperarm_fr_base_l, loc=<0 0 0>, rot=<0 0 0 1>}
 deform_upperarm_up_base_l {parent=upperarm_twist_01_l, loc=<-8.891 0.4825 3.9517>, rot=<0 0 0 -1>}
 deform_upperarm_up_l {parent=deform_upperarm_up_base_l, loc=<0 0 0>, rot=<0 0 0 1>}
 lowerarm_l {parent=upperarm_l, loc=<26.465 -0.8345 0.259>, rot=<0.0036 0.0161 -0.2298 -0.9731>}
 base_ChestFur_03_l {parent=clavicle_l, loc=<1.7011 0.8288 3.4943>, rot=<0.2601 0.1052 -0.0394 -0.959>}
 dyn_ChestFur_03_l {parent=base_ChestFur_03_l, loc=<0 0 0>, rot=<0 0 0 1>}
 clavicle_r {parent=spine_05, loc=<5.4831 -0.1053 3.2946>, rot=<0.7289 0.0465 -0.6829 0.009>}
 upperarm_r {parent=clavicle_r, loc=<-9.8658 1.2136 0.4105>, rot=<0.0164 -0.3245 -0.0269 -0.9454>}
 upperarm_twist_01_r {parent=upperarm_r, loc=<-7.2135 0.1134 0>, rot=<0 0 0 -1>}
 deform_upperarm_bk_base_r {parent=upperarm_twist_01_r, loc=<6.4053 4.3746 -0.314>, rot=<0 0 0 -1>}
 deform_upperarm_bk_r {parent=deform_upperarm_bk_base_r, loc=<0 0 0>, rot=<0 0 0 1>}
 deform_upperarm_fr_base_r {parent=upperarm_twist_01_r, loc=<5.6312 -5.092 -0.3354>, rot=<0 0 0 -1>}
 deform_upperarm_fr_r {parent=deform_upperarm_fr_base_r, loc=<0 0 0>, rot=<0 0 0 1>}
 deform_upperarm_up_base_r {parent=upperarm_twist_01_r, loc=<8.891 -0.4825 -3.9517>, rot=<0 0 0 -1>}
 deform_upperarm_up_r {parent=deform_upperarm_up_base_r, loc=<0 0 0>, rot=<0 0 0 1>}
 lowerarm_r {parent=upperarm_r, loc=<-26.465 0.8345 -0.259>, rot=<0.0036 0.0161 -0.2298 -0.9731>}
 base_ChestFur_03_r {parent=clavicle_r, loc=<-1.7011 -0.8288 -3.494>, rot=<-0.1053 0.2609 0.9588 -0.0393>}
 dyn_ChestFur_03_r {parent=base_ChestFur_03_r, loc=<0 0 0>, rot=<0 0 0 1>}
 base_ChestFur_02_l {parent=spine_05, loc=<4.7206 5.8656 -4.3112>, rot=<0.214 -0.6299 -0.2431 -0.7059>}
 dyn_ChestFur_02_l {parent=base_ChestFur_02_l, loc=<0 0 0>, rot=<0 0 0 1>}
 base_ChestFur_02_r {parent=spine_05, loc=<4.7201 5.8656 4.3112>, rot=<-0.214 0.6299 -0.2432 -0.7059>}
 dyn_ChestFur_02_r {parent=base_ChestFur_02_r, loc=<0 0 0>, rot=<0 0 0 -1>}
 dyn_backpack_mover {parent=spine_05, loc=<0 0 0>, rot=<0 0 0 1>}
 clavicle_l_null {parent=spine_05, type=Null, loc=<5.5162 -0.1063 -3.2946>, rot=<0.0465 -0.7289 -0.009 -0.6829>}
 clavicle_r_null {parent=spine_05, type=Null, loc=<5.4831 -0.1053 3.2946>, rot=<0.7289 0.0465 -0.6829 0.009>}
]
variables [
 CelebrationEmoteNoRban {type=double, default=0}
 CelebrationEmoteRban {type=double, default=0}
 Crouching {type=double, default=0}
 Crouching_Bool {type=bool, default=False}
 FallingSwim {type=double, default=0}
 Floating {type=bool, default=False}
 Glider {type=double, default=0}
 Moving2d {type=bool, default=False}
 On Ground Alpha {type=double, default=0}
 Pistol {type=double, default=0}
 Skydiving {type=bool, default=False}
 Wearing BackblingAndCape {type=bool, default=False}
 playEmote {type=bool, default=False}
]
