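# Physics asset for the hair on the F_MED_LiftingRaysHat character's face-accessory mesh (variant 02), built entirely from capsule bodies.
# Three four-link back-hair chains hang from base bones, with left and right tips tied to the center tip by angular-free constraints.
# Front and side strands branch from head, base bones or back chains with narrow swing and locked twist.
# Head, neck, spine, clavicle, upper-arm and chest-fur capsules form unconstrained body shapes.
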PhysicsAsset /BRCosmetics/Characters/Player/Female/Medium/Bodies/F_MED_LiftingRaysHat/Meshes/Parts/F_MED_LiftingRaysHat_FaceAcc_Physics_02
bodies {
 base_hairB_C_01 [{shapeType=Capsule, length=0.505, radius=0.505, shapeName=base_hairB_C_01_capsule, rotation=<rot 0 -90 0>}]
 base_hairB_L_01 [{shapeType=Capsule, length=0.505, radius=0.505, shapeName=base_hairB_L_01_capsule, rotation=<rot 0 -90 0>}]
 base_hairB_R_01 [{shapeType=Capsule, length=0.505, radius=0.505, shapeName=base_hairB_R_01_capsule, rotation=<rot 0 -90 0>}]
 base_hair_fr_a_r [{shapeType=Capsule, length=0.505, radius=0.505, shapeName=base_hair_fr_a_r_capsule, rotation=<rot 0 -90 0>}]
 base_hair_sd_g_01_r [{
   shapeType=Capsule
   length=0.6944
   radius=1.5
   shapeName=base_hair_sd_g_01_r_capsule
   center=<1.1098 0.3704 -0.3284>
   rotation=<rot -25.0762 88.3409 -65.8509>
  }]
 base_hair_sd_g_02_r [{
   shapeType=Capsule
   length=0.005
   radius=1.5
   shapeName=base_hair_sd_g_02_r_capsule
   center=<0.6001 0.8161 0.4388>
   rotation=<rot 0 107.904 -61.9743>
  }]
 base_hair_sd_g_r [{shapeType=Capsule, length=0.005, radius=1.8, shapeName=base_hair_sd_g_r_capsule, center=<-0.3339 0.635 0>, rotation=<rot 0 -90 0>}]
 base_sd_extra_l [{shapeType=Capsule, length=0.505, radius=0.505, shapeName=base_sd_extra_l_capsule, rotation=<rot 0 -90 0>}]
 clavicle_l [{shapeType=Capsule, length=10.6, radius=7, shapeName=clavicle_l_capsule, center=<6.2526 -0.0283 -2.2573>, rotation=<rot 0 -90 90>}]
 clavicle_r [
  {shapeType=Capsule, length=10.6, radius=5, shapeName=clavicle_l_capsule, center=<-6.2526 -0.1462 2.2573>, rotation=<rot 0 -90 90>}
  {shapeType=Capsule, length=3.8, radius=4.2, shapeName=clavicle_r_capsule, center=<-6.1037 -0.1074 -2.9779>, rotation=<rot 90 0 0>}
 ]
 dyn_ChestFur_03_l [{
   shapeType=Capsule
   length=2.1393
   radius=4.1911
   shapeName=dyn_ChestFur_03_r_capsule
   center=<4.7046 0.2401 -1.7511>
   rotation=<rot 81.5197 -179.9987 -179.9987>
  }]
 dyn_hairB_C_01 [{shapeType=Capsule, length=6.0314, radius=1, shapeName=dyn_hairB_C_01_capsule, center=<3.9112 0 0>, rotation=<rot 90 0 0>}]
 dyn_hairB_C_02 [{shapeType=Capsule, length=3.1828, radius=2.5094, shapeName=dyn_hairB_C_02_capsule, center=<3.855 0 0>, rotation=<rot 90 0 0>}]
 dyn_hairB_C_03 [{shapeType=Capsule, length=3.4505, radius=2.5764, shapeName=dyn_hairB_C_03_capsule, center=<4.6965 0.688 0>, rotation=<rot 90 10.4957 0>}]
 dyn_hairB_C_04 [{shapeType=Capsule, length=4.7663, radius=2.3534, shapeName=dyn_hairB_C_04_capsule, center=<3.7478 1.1023 0>}]
 dyn_hairB_L_01 [{shapeType=Capsule, length=6.1739, radius=1, shapeName=dyn_hairB_L_01_capsule, center=<3.7963 0 0>, rotation=<rot 90 0 0>}]
 dyn_hairB_L_02 [{shapeType=Capsule, length=4.5647, radius=2.5402, shapeName=dyn_hairB_L_02_capsule, center=<4.1798 -3.6094 0>, rotation=<rot 90 0 0>}]
 dyn_hairB_L_03 [{
   shapeType=Capsule
   length=3.4023
   radius=3.2924
   shapeName=dyn_hairB_L_03_capsule
   center=<3.4327 -0.3613 1.1568>
   rotation=<rot -0.1388 0.001 0>
  }]
 dyn_hairB_L_04 [{
   shapeType=Capsule
   length=4.0661
   radius=2.9276
   shapeName=dyn_hairB_L_04_capsule
   center=<2.8312 -0.0222 0.3155>
   rotation=<rot -0.6197 0.0001 0>
  }]
 dyn_hairB_R_01 [{shapeType=Capsule, length=6.5288, radius=1, shapeName=dyn_hairB_R_01_capsule, center=<3.6783 0 0>, rotation=<rot 90 0 0>}]
 dyn_hairB_R_02 [
  {shapeType=Capsule, length=4, radius=3, shapeName=dyn_hairB_R_02_capsule, center=<3.899 -0.1481 0>, rotation=<rot 90 0 0>}
  {shapeType=Capsule, length=8.5, radius=2.5, shapeName=dyn_hairB_R_02_capsule, center=<4.0987 6.3756 -5.5236>, rotation=<rot 90 -41.3658 0>}
 ]
 dyn_hairB_R_03 [{
   shapeType=Capsule
   length=4.4769
   radius=3.6838
   shapeName=dyn_hairB_R_03_capsule
   center=<2.732 -1.0623 0>
   rotation=<rot 0.5212 -179.983 179.9996>
  }]
 dyn_hairB_R_04 [{
   shapeType=Capsule
   length=6.28
   radius=2.9343
   shapeName=dyn_hairB_R_04_capsule
   center=<1.0256 -1.3265 0.0102>
   rotation=<rot -0.5984 0.0001 0>
  }]
 dyn_hair_fr_a_01_r [{
   shapeType=Capsule
   length=1.8401
   radius=1.3197
   shapeName=dyn_hair_fr_a_01_r_capsule
   center=<2.0366 0.812 -0.3293>
   rotation=<rot 7.3669 -90 5.9791>
  }]
 dyn_hair_fr_a_02_r [{
   shapeType=Capsule
   length=2.4092
   radius=1
   shapeName=dyn_hair_fr_a_02_r_capsule
   center=<1.4484 0.325 0.6602>
   rotation=<rot -2.092 74.6342 -66.9872>
  }]
 dyn_hair_sd_a_01_l [{
   shapeType=Capsule
   length=2.0417
   radius=1.8257
   shapeName=dyn_hair_sd_a_01_l_capsule
   center=<2.2279 0.2466 -0.2808>
   rotation=<rot 0 -95.2782 -79.3738>
  }]
 dyn_hair_sd_a_01_r [{
   shapeType=Capsule
   length=2.4066
   radius=1.0623
   shapeName=dyn_hair_sd_a_01_r_capsule
   center=<1.4334 0.1818 -0.2731>
   rotation=<rot -1.0794 -90.0806 -0.8787>
  }]
 dyn_hair_sd_a_02_r [{shapeType=Capsule, length=2, radius=1, shapeName=dyn_hair_sd_a_02_r_capsule, center=<1.7212 0.1731 0>, rotation=<rot 90 0 0>}]
 dyn_hair_sd_a_03_r [{
   shapeType=Capsule
   length=2.5
   radius=1
   shapeName=dyn_hair_sd_a_03_r_capsule
   center=<2.9656 -0.1193 0.0866>
   rotation=<rot 77.7578 -179.9998 -166.8766>
  }]
 dyn_hair_sd_b_01_l [{
   shapeType=Capsule
   length=3
   radius=1
   shapeName=dyn_hair_sd_b_01_l_capsule
   center=<2.4132 -0.3054 -0.1374>
   rotation=<rot 84.0266 90.0001 -179.9999>
  }]
 dyn_hair_sd_b_01_r [{shapeType=Capsule, length=3, radius=1, shapeName=dyn_hair_sd_b_01_r_capsule, center=<1.2768 0 0>, rotation=<rot 90 0 0>}]
 dyn_hair_sd_b_02_l [
  {shapeType=Capsule, length=2.5, radius=1, shapeName=dyn_hair_sd_b_02_l_capsule, center=<0.6153 -0.2085 -0.257>, rotation=<rot 90 -90 0>}
  {
   shapeType=Capsule
   length=1.5
   radius=1
   shapeName=dyn_hair_sd_b_02_l_capsule
   center=<2.3482 0.2401 -1.3176>
   rotation=<rot 61.629 0 -12.7623>
  }
 ]
 dyn_hair_sd_b_02_r [{
   shapeType=Capsule
   length=2.5
   radius=1
   shapeName=dyn_hair_sd_b_02_r_capsule
   center=<2.1007 -0.0003 0.7119>
   rotation=<rot 76.2203 -179.982 -179.9661>
  }]
 dyn_hair_sd_c_01_r [{shapeType=Capsule, length=4, radius=1, shapeName=dyn_hair_sd_c_01_r_capsule, center=<1.5869 0 0.5193>, rotation=<rot 90 0 0>}]
 dyn_hair_sd_c_02_r [{
   shapeType=Capsule
   length=2.5
   radius=0.5
   shapeName=dyn_hair_sd_c_02_r_capsule
   center=<1.7606 0.0767 -0.356>
   rotation=<rot 67.9453 0.0001 -3.5346>
  }]
 dyn_hair_sd_d_01_l [{shapeType=Capsule, length=2, radius=1, shapeName=dyn_hair_sd_d_01_l_capsule, center=<0.7698 0 0>}]
 dyn_hair_sd_d_01_r [{shapeType=Capsule, length=2.2325, radius=1.2994, shapeName=dyn_hair_sd_d_01_r_capsule, center=<1.3259 0 0>, rotation=<rot 90 0 0>}]
 dyn_hair_sd_d_02_l [{shapeType=Capsule, length=1.5, radius=1, shapeName=dyn_hair_sd_d_02_l_capsule}]
 dyn_hair_sd_d_02_r [{
   shapeType=Capsule
   length=3.0454
   radius=1.1
   shapeName=dyn_hair_sd_d_02_r_capsule
   center=<1.2881 -0.0564 0.1263>
   rotation=<rot 0 84.8557 -86.1156>
  }]
 dyn_hair_sd_d_03_l [{shapeType=Capsule, length=2.5, radius=1, shapeName=dyn_hair_sd_d_03_l_capsule, center=<0.9298 0 -0.5261>, rotation=<rot 66.3222 0 0>}]
 dyn_hair_sd_d_03_r [{
   shapeType=Capsule
   length=2.6095
   radius=0.9
   shapeName=dyn_hair_sd_d_03_r_capsule
   center=<1.2563 0.5369 -0.9823>
   rotation=<rot 0 -66.0146 -52.7398>
  }]
 dyn_hair_sd_e_01_l [{shapeType=Capsule, length=1, radius=1.5, shapeName=dyn_hair_sd_e_01_l_capsule, center=<0.7509 0.0797 0.8344>, rotation=<rot 90 -90 0>}]
 dyn_hair_sd_e_02_l [{shapeType=Capsule, length=0.5, radius=1.5, shapeName=dyn_hair_sd_e_02_l_capsule, center=<0 0 0.3585>, rotation=<rot 90 -90 0>}]
 dyn_hair_sd_e_03_l [{shapeType=Capsule, length=2, radius=1, shapeName=dyn_hair_sd_e_03_l_capsule, center=<-0.0115 -0.0892 0>, rotation=<rot 90 -7.37 0>}]
 dyn_hair_sd_f_02_r [{
   shapeType=Capsule
   length=2.5158
   radius=0.884
   shapeName=dyn_hair_sd_f_02_r_capsule
   center=<1.0119 -0.0198 -0.1184>
   rotation=<rot 0 -92.4438 -87.3515>
  }]
 dyn_hair_sd_f_03_r [{
   shapeType=Capsule
   length=0.505
   radius=0.505
   shapeName=dyn_hair_sd_f_03_r_capsule
   center=<0.7609 0.0042 0.5966>
   rotation=<rot 0 89.817 -51.2689>
  }]
 dyn_sd_extra_l [{shapeType=Capsule, length=0.505, radius=1, shapeName=dyn_sd_extra_l_capsule, center=<0.3044 0 0>, rotation=<rot 0 -90 0>}]
 head [{shapeType=Capsule, length=2.6403, radius=7.3597, center=<2.5678 1.1032 0>, rotation=<rot 90 0 0>}]
 neck_01 [{shapeType=Capsule, length=12.5, radius=5, shapeName=neck_01_capsule, center=<1.6583 0.6206 0>}]
 spine_04 [
  {shapeType=Capsule, length=17.8, radius=9.1, shapeName=spine_04_capsule, center=<4.0135 1.6435 0>, rotation=<rot 0 -90 0>}
  {shapeType=Capsule, length=18.5, radius=9, shapeName=spine_04_capsule, center=<8.6672 0.6315 0>}
 ]
 spine_05 [{shapeType=Capsule, length=17.2, radius=8, shapeName=spine_05_capsule, center=<2.8857 0.3837 0>, rotation=<rot 0 -90 0>}]
 upperarm_l [{shapeType=Capsule, length=13.5146, radius=5.3875, shapeName=upperarm_l_capsule, center=<11.2314 -0.3627 0>, rotation=<rot 90 0 0>}]
 upperarm_r [{shapeType=Capsule, length=13.5146, radius=5.3875, shapeName=upperarm_l_capsule, center=<-11.2314 0.8286 0>, rotation=<rot 90 0 0>}]
}
constraints [
 {bone1Name=dyn_hairB_C_01, bone2Name=base_hairB_C_01, swing1LimitDegrees=10, swing2LimitDegrees=15, twistLimitDegrees=1}
 {bone1Name=dyn_hairB_C_02, bone2Name=dyn_hairB_C_01, swing1LimitDegrees=30, swing2LimitDegrees=30, twistLimitDegrees=1}
 {bone1Name=dyn_hairB_C_03, bone2Name=dyn_hairB_C_02, swing1LimitDegrees=40, swing2LimitDegrees=25, twistLimitDegrees=1, twistMotion=Locked}
 {bone1Name=dyn_hairB_C_04, bone2Name=dyn_hairB_C_03, swing1LimitDegrees=30, swing2LimitDegrees=5, twistLimitDegrees=1, twistMotion=Locked}
 {bone1Name=dyn_hairB_L_01, bone2Name=base_hairB_L_01, swing1LimitDegrees=10, swing2LimitDegrees=15, twistLimitDegrees=1}
 {bone1Name=dyn_hairB_L_02, bone2Name=dyn_hairB_L_01, swing1LimitDegrees=30, swing2LimitDegrees=30, twistLimitDegrees=1}
 {bone1Name=dyn_hairB_L_03, bone2Name=dyn_hairB_L_02, swing1LimitDegrees=40, swing2LimitDegrees=25, twistLimitDegrees=1}
 {bone1Name=dyn_hairB_L_04, bone2Name=dyn_hairB_L_03, swing1LimitDegrees=20, swing2LimitDegrees=6, twistLimitDegrees=1}
 {bone1Name=dyn_hairB_R_01, bone2Name=base_hairB_R_01, swing1LimitDegrees=15, swing2LimitDegrees=15, twistLimitDegrees=1, twistMotion=Locked}
 {bone1Name=dyn_hairB_R_02, bone2Name=dyn_hairB_R_01, swing1LimitDegrees=30, swing2LimitDegrees=30, twistLimitDegrees=1, twistMotion=Locked}
 {bone1Name=dyn_hairB_R_03, bone2Name=dyn_hairB_R_02, swing1LimitDegrees=45, swing2LimitDegrees=25, twistLimitDegrees=1, twistMotion=Locked}
 {bone1Name=dyn_hairB_R_04, bone2Name=dyn_hairB_R_03, swing1LimitDegrees=36, swing2LimitDegrees=5, twistLimitDegrees=1}
 {
  bone1Name=dyn_hairB_L_04
  bone2Name=dyn_hairB_C_04
  swing1LimitDegrees=45
  swing1Motion=Free
  swing2LimitDegrees=45
  swing2Motion=Free
  twistLimitDegrees=45
  twistMotion=Free
 }
 {
  bone1Name=dyn_hairB_R_04
  bone2Name=dyn_hairB_C_04
  swing1LimitDegrees=45
  swing1Motion=Free
  swing2LimitDegrees=45
  swing2Motion=Free
  twistLimitDegrees=45
  twistMotion=Free
 }
 {bone1Name=dyn_hair_sd_b_01_l, bone2Name=head, swing1LimitDegrees=50, swing2LimitDegrees=5, twistLimitDegrees=45, twistMotion=Locked}
 {
  bone1Name=dyn_hair_sd_b_02_l
  bone2Name=dyn_hair_sd_b_01_l
  swing1LimitDegrees=50
  swing2LimitDegrees=10
  twistLimitDegrees=45
  twistMotion=Locked
 }
 {bone1Name=dyn_hair_sd_b_01_r, bone2Name=head, swing1LimitDegrees=10, swing2LimitDegrees=22, twistLimitDegrees=45, twistMotion=Locked}
 {
  bone1Name=dyn_hair_sd_b_02_r
  bone2Name=dyn_hair_sd_b_01_r
  swing1LimitDegrees=10
  swing2LimitDegrees=15
  twistLimitDegrees=45
  twistMotion=Locked
 }
 {
  bone1Name=dyn_hair_sd_d_01_l
  bone2Name=dyn_hairB_L_02
  swing1LimitDegrees=10
  swing2LimitDegrees=20
  twistLimitDegrees=45
  twistMotion=Locked
 }
 {
  bone1Name=dyn_hair_sd_d_02_l
  bone2Name=dyn_hair_sd_d_01_l
  swing1LimitDegrees=10
  swing2LimitDegrees=20
  twistLimitDegrees=45
  twistMotion=Locked
 }
 {
  bone1Name=dyn_hair_sd_d_03_l
  bone2Name=dyn_hair_sd_d_02_l
  swing1LimitDegrees=10
  swing2LimitDegrees=20
  twistLimitDegrees=45
  twistMotion=Locked
 }
 {
  bone1Name=dyn_hair_sd_e_01_l
  bone2Name=dyn_hairB_L_02
  swing1LimitDegrees=10
  swing2LimitDegrees=20
  twistLimitDegrees=45
  twistMotion=Locked
 }
 {
  bone1Name=dyn_hair_sd_e_02_l
  bone2Name=dyn_hair_sd_e_01_l
  swing1LimitDegrees=10
  swing2LimitDegrees=20
  twistLimitDegrees=45
  twistMotion=Locked
 }
 {
  bone1Name=dyn_hair_sd_e_03_l
  bone2Name=dyn_hair_sd_e_02_l
  swing1LimitDegrees=10
  swing2LimitDegrees=20
  twistLimitDegrees=45
  twistMotion=Locked
 }
 {bone1Name=dyn_hair_sd_a_01_r, bone2Name=head, swing1LimitDegrees=10, swing2LimitDegrees=20, twistLimitDegrees=45, twistMotion=Locked}
 {
  bone1Name=dyn_hair_sd_a_02_r
  bone2Name=dyn_hair_sd_a_01_r
  swing1LimitDegrees=10
  swing2LimitDegrees=20
  twistLimitDegrees=45
  twistMotion=Locked
 }
 {
  bone1Name=dyn_hair_sd_a_03_r
  bone2Name=dyn_hair_sd_a_01_r
  swing1LimitDegrees=10
  swing2LimitDegrees=20
  twistLimitDegrees=45
  twistMotion=Locked
 }
 {
  bone1Name=dyn_hair_fr_a_01_r
  bone2Name=base_hair_fr_a_r
  swing1LimitDegrees=10
  swing2LimitDegrees=20
  twistLimitDegrees=45
  twistMotion=Locked
 }
 {
  bone1Name=dyn_hair_fr_a_02_r
  bone2Name=dyn_hair_fr_a_01_r
  swing1LimitDegrees=10
  swing2LimitDegrees=20
  twistLimitDegrees=45
  twistMotion=Locked
 }
 {
  bone1Name=dyn_hair_sd_c_02_r
  bone2Name=dyn_hair_sd_c_01_r
  swing1LimitDegrees=10
  swing2LimitDegrees=12
  twistLimitDegrees=45
  twistMotion=Locked
 }
 {
  bone1Name=dyn_hair_sd_d_02_r
  bone2Name=dyn_hair_sd_d_01_r
  swing1LimitDegrees=7
  swing2LimitDegrees=12
  twistLimitDegrees=45
  twistMotion=Locked
 }
 {
  bone1Name=dyn_hair_sd_d_03_r
  bone2Name=dyn_hair_sd_d_02_r
  swing1LimitDegrees=10
  swing2LimitDegrees=20
  twistLimitDegrees=45
  twistMotion=Locked
 }
 {
  bone1Name=base_hair_sd_g_01_r
  bone2Name=base_hair_sd_g_r
  swing1LimitDegrees=35
  swing2LimitDegrees=10
  twistLimitDegrees=45
  twistMotion=Locked
 }
 {
  bone1Name=base_hair_sd_g_02_r
  bone2Name=base_hair_sd_g_01_r
  swing1LimitDegrees=20
  swing2LimitDegrees=10
  twistLimitDegrees=45
  twistMotion=Locked
 }
 {
  bone1Name=dyn_hair_sd_f_03_r
  bone2Name=dyn_hair_sd_f_02_r
  swing1LimitDegrees=10
  swing2LimitDegrees=10
  twistLimitDegrees=45
  twistMotion=Locked
 }
 {bone1Name=dyn_hair_sd_a_01_l, bone2Name=head, swing1LimitDegrees=5, swing2LimitDegrees=10, twistLimitDegrees=45, twistMotion=Locked}
 {
  bone1Name=dyn_hair_sd_c_01_r
  bone2Name=dyn_sd_extra_l
  swing1LimitDegrees=5
  swing2LimitDegrees=25
  twistLimitDegrees=45
  twistMotion=Locked
 }
 {
  bone1Name=dyn_hair_sd_d_01_r
  bone2Name=dyn_sd_extra_l
  swing1LimitDegrees=5
  swing2LimitDegrees=15
  twistLimitDegrees=45
  twistMotion=Locked
 }
 {
  bone1Name=dyn_sd_extra_l
  bone2Name=base_sd_extra_l
  swing1LimitDegrees=8
  swing2LimitDegrees=2
  swing2Motion=Locked
  twistLimitDegrees=45
  twistMotion=Locked
 }
 {
  bone1Name=dyn_hair_sd_f_02_r
  bone2Name=dyn_hairB_R_02
  swing1LimitDegrees=35
  swing2LimitDegrees=10
  twistLimitDegrees=45
  twistMotion=Locked
 }
 {bone1Name=base_hair_sd_g_r, bone2Name=dyn_hairB_R_02, swing1LimitDegrees=35, swing2LimitDegrees=10, twistLimitDegrees=45, twistMotion=Locked}
]
